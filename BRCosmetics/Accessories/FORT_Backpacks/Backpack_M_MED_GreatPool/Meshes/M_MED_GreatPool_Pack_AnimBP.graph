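# Animation Blueprint for the GreatPool cosmetic backpack that procedurally rumbles the pack and its four exhausts.
# Each update, a sine-of-game-time macro builds rotators, with separate frequency and amplitude for top and bottom exhausts and identical left/right motion.
# Chained Modify Bone nodes add these rotations to the input pose.
# When character customization loads, a gender check sets isMale, which enables a small translation offset on the pack bone.
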AnimBlueprint /BRCosmetics/Accessories/FORT_Backpacks/Backpack_M_MED_GreatPool/Meshes/M_MED_GreatPool_Pack_AnimBP
target_skeleton=/Game/Accessories/FORT_Backpacks/Backpack_M_MED_GreatPool/Meshes/M_MED_GreatPool_Pack_Skeleton
graphs [
 AnimGraph {nodes [
   n1 = "Transform (Modify) Bone - Bone: pack"!(ComponentPose=Local To Component(LocalPose=Input Pose()), Alpha, Rotation=Get general rotator() -> (general rotator:Rotator)) -> (Pose) {
    rotationMode=BMM_Additive
    rotationSpace=BCS_BoneSpace
    translationMode=BMM_Additive
    translationSpace=BCS_WorldSpace
    alphaBoolBlend {blendOption=Linear}
    boneToModify {boneName=pack}
   }
   n2 = "Transform (Modify) Bone - Bone: exhaust_top_L_1"!(ComponentPose=@n1, Alpha=1, Rotation=Get topExhaust left rotator() -> (topExhaust left rotator:Rotator)) -> (Pose) {
    rotationMode=BMM_Additive
    rotationSpace=BCS_BoneSpace
    translationMode=BMM_Additive
    translationSpace=BCS_WorldSpace
    alphaBoolBlend {blendOption=Linear}
    boneToModify {boneName=exhaust_top_L_1}
   }
   n3 = "Transform (Modify) Bone - Bone: exhaust_bot_L_1"!(ComponentPose=@n2, Alpha=1, Rotation=Get botExhaust left rotator() -> (botExhaust left rotator:Rotator)) -> (Pose) {
    rotationMode=BMM_Additive
    rotationSpace=BCS_BoneSpace
    translationMode=BMM_Additive
    translationSpace=BCS_WorldSpace
    alphaBoolBlend {blendOption=Linear}
    boneToModify {boneName=exhaust_bot_L_1}
   }
   n4 = "Transform (Modify) Bone - Bone: exhaust_top_R_1"!(ComponentPose=@n3, Alpha=1, Rotation=Get topExhaust right rotator() -> (topExhaust right rotator:Rotator)) -> (Pose) {
    rotationMode=BMM_Additive
    rotationSpace=BCS_BoneSpace
    translationMode=BMM_Additive
    translationSpace=BCS_WorldSpace
    alphaBoolBlend {blendOption=Linear}
    boneToModify {boneName=exhaust_top_R_1}
   }
   n5 = "Transform (Modify) Bone - Bone: exhaust_bot_R_1"!(ComponentPose=@n4, Alpha=1, Rotation=Get botExhaust right rotator() -> (botExhaust right rotator:Rotator)) -> (Pose) {
    rotationMode=BMM_Additive
    rotationSpace=BCS_BoneSpace
    translationMode=BMM_Additive
    translationSpace=BCS_WorldSpace
    alphaBoolBlend {blendOption=Linear}
    boneToModify {boneName=exhaust_bot_R_1}
   }
   n6 = "Transform (Modify) Bone - Bone: pack"!(ComponentPose=@n5, bAlphaBoolEnabled:bool=Get isMale() -> (isMale:bool), Translation:Vector="0,-0.8,0") -> (Pose) {
    alphaInputType=Bool
    rotationMode=BMM_Additive
    rotationSpace=BCS_BoneSpace
    translationMode=BMM_Additive
    translationSpace=BCS_BoneSpace
    alphaBoolBlend {blendOption=Linear}
    boneToModify {boneName=pack}
    translation {y=-0.8}
   }
   Output Pose(Result=Component To Local(ComponentPose=@n6)) {layerGroup=None}
  ]}
 EventGraph {nodes [
   Event BlueprintUpdateAnimation()
   > Set general rotator(general rotator:Rotator=Rotator Rumble!(frequency:double=40, amplitude:double=0.08, time offset:double, coefficient:Vector="1,0,0") -> (rotatorOutput:Rotator=~) {is=title K2Node_MacroInstance}) -> (Output_Get:Rotator)
   n1 = Get amplitude top() -> (amplitude top:double)
   n2 = Get frequency top() -> (frequency top:double)
   > Set topExhaust left rotator(topExhaust left rotator:Rotator=Rotator Rumble(frequency=@n2, amplitude=@n1, time offset=1, coefficient="0,1,1") -> (rotatorOutput=~)) -> (Output_Get:Rotator)
   > Set topExhaust right rotator(topExhaust right rotator:Rotator=Rotator Rumble(frequency=@n2, amplitude=@n1, time offset=1, coefficient="0,1,1") -> (rotatorOutput=~)) -> (Output_Get:Rotator)
   n3 = Get amplitude bot() -> (amplitude bot:double)
   n4 = Get frequency bot() -> (frequency bot:double)
   > Set botExhaust left rotator(botExhaust left rotator:Rotator=Rotator Rumble(frequency=@n4, amplitude=@n3, time offset=1.2, coefficient="0,1,1") -> (rotatorOutput=~)) -> (Output_Get:Rotator)
   > Set botExhaust right rotator(botExhaust right rotator:Rotator=Rotator Rumble(frequency=@n4, amplitude=@n3, time offset=1.2, coefficient="0,1,1") -> (rotatorOutput=~)) -> (Output_Get:Rotator)
   Event OnCharacterCustomizationLoaded()
   > n5 = Set Owning Pawn(Owning Pawn:Pawn Object Reference=TryGetFortPawn()) -> (Output_Get:Pawn Object Reference)
   > Is Valid(InputObject=@n5)
   > n6 = FindGender!(execute, Object:Object Reference=Get Owning Pawn() -> (Owning Pawn:Pawn Object Reference)) -> (True:Exec, False:Exec, OutGender:EFortCustomGender Enum=Invalid)
   > Set isMale(isMale:bool="Equal (Enum)"(A:EFortCustomGender Enum=@n6, B:EFortCustomGender Enum=Male)) -> (Output_Get:bool)
   Event BlueprintInitializeAnimation()
   > Set Skel_Mesh(Skel_Mesh=GetOwnerSkeletalMeshComponent())
   TryGetPawnOwner()
  ]}
 floatSinFunc {nodes [
   n1 = Inputs!() -> (time:double=~, time offset:double=~, frequency:double=~, amplitude:double=~, coefficient:double=~)
   n2 = float + float(A=@n1, B=@n1.time offset)
   n3 = MultiplyByPi!(Value:double="Sin (Radians)"(A=float * float(A=@n2, B=@n1.frequency))) -> (ReturnValue:double)
   Outputs!(floatOutput:double=float * float(A=float * float(A=@n3, B=@n1.amplitude), B=@n1.coefficient))
  ]}
 rotatorRumble {nodes [
   n1 = Inputs!() -> (frequency, amplitude, time offset=~, coefficient:Vector=~)
   n2 = BreakVector(InVec=@n1.coefficient)
   n3 = GetGameTimeInSeconds()
   Outputs!(rotatorOutput:Rotator=MakeRotator(
     Roll=Float Sin Func!(time:double=@n3, time offset:double=@n1.time offset, frequency:double=@n1, amplitude:double=@n1.amplitude, coefficient:double=@n2) -> (floatOutput:double=~) {is=title K2Node_MacroInstance}
     Pitch=Float Sin Func(time=@n3, time offset=@n1.time offset, frequency=@n1, amplitude=@n1.amplitude, coefficient=@n2.Y) -> (floatOutput=~)
     Yaw=Float Sin Func(time=@n3, time offset=@n1.time offset, frequency=@n1, amplitude=@n1.amplitude, coefficient=@n2.Z) -> (floatOutput=~)
    ))
  ]}
]
variables [
 amplitude bot {type=number, default=0.15}
 amplitude top {type=number, default=0.2}
 botExhaust left rotator {type=Rotator}
 botExhaust right rotator {type=Rotator}
 frequency bot {type=number, default=30}
 frequency top {type=number, default=35}
 general rotator {type=Rotator}
 isMale {type=boolean}
 Owning Pawn {type=Pawn, default=null}
 Skel_Mesh {type=SkeletalMeshComponent}
 topExhaust left rotator {type=Rotator}
 topExhaust right rotator {type=Rotator}
]
